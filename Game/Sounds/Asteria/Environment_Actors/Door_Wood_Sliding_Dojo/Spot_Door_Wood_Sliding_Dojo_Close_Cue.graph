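# Sound cue for the closing of a sliding wooden dojo door, routed to the doors sound class with shared door attenuation.
# A Random node picks between two equally weighted close waves, feeding a Quality Level node and then a root Modulator.
# All four quality inputs share that Random node, which is excluded from branch culling.
# Overridden concurrency prevents new instances once the limit is reached.
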SoundCue /Game/Sounds/Asteria/Environment_Actors/Door_Wood_Sliding_Dojo/Spot_Door_Wood_Sliding_Dojo_Close_Cue
root=Modulator_1
nodes [
 {id=Modulator_1, children [QualityLevel_0]}
 {id=QualityLevel_0, children [Random_2,Random_2,Random_2,Random_2]}
 {id=Random_2, children [WavePlayer_5,WavePlayer_6], properties {bShouldExcludeFromBranchCulling=true, weights [1,1]}}
 {
  id=WavePlayer_5
  properties {soundWaveAssetPtr=&/Spot_Door_Wood_Sliding_Dojo_Close_02}
 }
 {
  id=WavePlayer_6
  properties {soundWaveAssetPtr=&/Spot_Door_Wood_Sliding_Dojo_Close_03}
 }
]
properties {
 attenuationSettings=/Game/Sounds/Attenuation/Fort_Doors_Att
 bOverrideConcurrency=true
 maxDistance=2400
 soundClassObject=&Fort_Doors
 volumeMultiplier=1.2
 attenuationOverrides {falloffDistance=1600}
 concurrencyOverrides {resolutionRule=PreventNew}
}
